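# Unlit masked material for the tire rubble props, apparently an invisible surface for jump collision meshes, judging by its name.
# The graph holds a single Constant node at its default value, with no textures or parameters.
# It is flagged for Nanite and instanced static meshes, and ray-traced shadow casting is off.
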Material /Game/Environments/Props/Rubble/Materials/M_Tire_JumpCollision_Clear
graph {
 properties {
  bCastRayTracedShadows=false
  bUsedWithInstancedStaticMeshes=true
  bUsedWithNanite=true
  blendMode=BLEND_Masked
  pixelDepthOffsetMode=PDOM_Legacy
  shadingModel=MSM_Unlit
  translucencyPass=MTP_BeforeDOF
  usedShadingModels=Unlit
 }
 nodes [Constant()]
}
